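# NS_DyePacket is the Niagara effect for the Dye Packet vehicle boost cosmetic, trailing from skeletal mesh boost sockets.
# One emitter draws five layered ribbons, with curve-driven widths, offsets and colors over particle life.
# Two sprite emitters partition short-lived SubUV particles across the ports, tagging spawn groups for separate renderers.
# Painted-variant modules choose colors per paint, while bIsPainted switches some ribbons' emissive and alpha scaling and bIsFrontendPreview adjusts one sprite velocity.
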NiagaraSystem /VehicleCosmetics/Boosts/Boost_DyePacket/FX/NS_DyePacket
dependencies {
 usedDataInterfaces [
  &NiagaraDataInterfaceArrayColor
  &NiagaraDataInterfaceCurve
  &NiagaraDataInterfaceColorCurve
  &NiagaraDataInterfaceSpriteRendererInfo
  &NiagaraDataInterfaceSkeletalMesh
 ]
 usedDynamicInputs []
 usedModules [
  &self:SystemSpawnScript.SystemScriptSource.SystemScriptGraph.NiagaraNodeAssignment_0.SetVariables_#1
  &SystemState
 ]
}
emitters {
 NE_MainRibbon_Boost {
  emitterUpdateScript [
   &EmitterState()
   &VC_FilteredSocketCount()
   &VC_RateQualityHandler(Epic Spawn Rate=30, High Spawn Rate=30, Low Spawn Rate=30, Use Custom Rate?=true)
   &VC_SpawnRibbon_v1(Number of Ribbons=$Emitter.PortCount, SpawnRate=$Emitter.Spawn Rate)
  ]
  particleSpawnScript [
   &VC_PaintedVariantsOverride(
    R1 Cobalt:LinearColor=<rgba 0 0.0185 1 0.8>
    R1 Crimson:LinearColor=<rgba 0.4635 0 0.0126 0.8>
    R1 DefaultColor=<rgba 0.0123 0.3095 0.9647 0.5>
    R1 ForestGreen:LinearColor=<rgba 0 0.5 0.0253 0.7>
    R1 Grey:LinearColor=<rgba 0.1302 0.1302 0.1302 0.8>
    Use R1 Painted Overrides=true
    Use R2 Painted Overrides=true
   ) {enabled=false}
   &VC_InitializeRibbon_v1(
    Custom Offset=true
    Enable Skeletal Position=true
    Lifetime=0.8
    Local Vector Offset=&AddVectorToPosition(Position=<20 0 0>, Vector=Linked() {linkedVariable {name=User.DTCarOffset, type=NiagaraPosition}})
    Offset Amount=1
   )
   &AddVelocity(Velocity=<-1400 0 0>) {module=AddVelocity001}
  ]
  particleUpdateScript [
   &VC_PaintedVariantsOverride(
    R1 Black:LinearColor=<rgba 0.01 0.01 0.01 0.85>
    R1 BurntSienna:LinearColor=<rgba 0.0642 0.0198 0.0079 0.7>
    R1 Cobalt:LinearColor=<rgba 0 0.0113 0.6094 0.7>
    R1 Crimson:LinearColor=<rgba 0.4635 0 0.0126 0.5>
    R1 DefaultColor=<rgba 0.0123 0.3095 0.9647 0.5>
    R1 ForestGreen:LinearColor=<rgba 0 0.5 0.0253 0.6>
    R1 Grey:LinearColor=<rgba 0.147 0.147 0.147 0.8>
    R1 Lime:LinearColor=<rgba 0.2083 0.4167 0 0.5>
    R1 Orange:LinearColor=<rgba 0.7292 0.1314 0.0195 0.5>
    R1 Pink:LinearColor=<rgba 0.4635 0.0555 0.3152 0.6>
    R1 Purple:LinearColor=<rgba 0.1901 0 0.9391 0.7>
    R1 Saffron:LinearColor=<rgba 0.5469 0.4193 0 0.39>
    R1 SkyBlue:LinearColor=<rgba 0 0.4332 0.651 0.5>
    R1 TitaniumWhite:LinearColor=&ColorFromCurve(ColorCurve=DataInterface() {propertyValues {
       AlphaCurve {keys [<at 0 1>,<at 0.3261 0.05>,<at 0.5495 0.2>,<at 0.7007 0.2>,<at 0.8175 0 RCIM_Cubic>]}
       BlueCurve {keys [<at 0 0.1846 RCIM_Cubic>,<at 0.2285 0.1336>,<at 0.4863 0.9274>,<at 0.6405 0.3248>]}
       GreenCurve {keys [<at 0 0.2083 RCIM_Cubic>,<at 0.2285 0.1336>,<at 0.4863 0.9274>,<at 0.6405 0.3248>]}
       RedCurve {keys [<at 0 0.2046 RCIM_Cubic>,<at 0.2285 0.1406>,<at 0.4863 0.9274>,<at 0.6405 0.3333>]}
      }})
    R2 Black:LinearColor=&ColorFromCurve(ColorCurve=DataInterface() {propertyValues {
       AlphaCurve {keys [<at 0 7>,<at 0.2232 1>,<at 0.3261 0.05>,<at 0.5495 0.2>,<at 0.7007 0.2>,<at 0.8474 0 RCIM_Cubic>]}
       BlueCurve {keys [<at 0 0.0194 RCIM_Cubic>,<at 0.1492 0.0333>,<at 0.2681 0.1104>,<at 0.4863 0.9274>,<at 0.6088 0.0305>]}
       GreenCurve {keys [<at 0 0.0194 RCIM_Cubic>,<at 0.1492 0.0333>,<at 0.2681 0.1104>,<at 0.4863 0.9274>,<at 0.6088 0.0305>]}
       RedCurve {keys [<at 0 0.0194 RCIM_Cubic>,<at 0.1492 0.0333>,<at 0.2681 0.1104>,<at 0.4863 0.9274>,<at 0.6088 0.0305>]}
      }})
    R2 BurntSienna:LinearColor=&ColorFromCurve(ColorCurve=DataInterface() {propertyValues {
       AlphaCurve {keys [<at 0 4>,<at 0.1907 1>,<at 0.3261 0.05>,<at 0.593 0.2>,<at 0.8474 0 RCIM_Cubic>]}
       BlueCurve {keys [<at 0 0.008 RCIM_Cubic>,<at 0.2839 0.0202>,<at 0.4863 0.2907>,<at 0.5956 0.008>]}
       GreenCurve {keys [<at 0 0.0203 RCIM_Cubic>,<at 0.2839 0.0323>,<at 0.4863 0.4394>,<at 0.5956 0.0203>]}
       RedCurve {keys [<at 0 0.0648 RCIM_Cubic>,<at 0.2839 0.0762>,<at 0.4863 0.9792>,<at 0.5956 0.0648>]}
      }})
    R2 Cobalt:LinearColor=&ColorFromCurve(ColorCurve=DataInterface() {propertyValues {
       AlphaCurve {keys [<at 0 7>,<at 0.1907 1>,<at 0.3261 0.05>,<at 0.5495 0.6>,<at 0.7007 0.2>,<at 0.8474 0 RCIM_Cubic>]}
       BlueCurve {keys [<at 0 0.6458 RCIM_Cubic>,<at 0.2575 0.9985>,<at 0.4863 1>,<at 0.6115 0.8177>]}
       GreenCurve {keys [<at 0 0.0119 RCIM_Cubic>,<at 0.2575 0.0367>,<at 0.4863 0.2997>,<at 0.6115 0.0527>]}
       RedCurve {keys [<at 0 0 RCIM_Cubic>,<at 0.2575 0.0186>,<at 0.4863 0.2865>,<at 0.6115 0.0383>]}
      }})
    R2 Crimson:LinearColor=&ColorFromCurve(ColorCurve=DataInterface() {propertyValues {
       AlphaCurve {keys [<at 0 7>,<at 0.1907 1>,<at 0.3261 0.05>,<at 0.5495 0.2>,<at 0.7007 0.2>,<at 0.8474 0 RCIM_Cubic>]}
       BlueCurve {keys [<at 0 0 RCIM_Cubic>,<at 0.1492 0.0104>,<at 0.4863 0.2995>,<at 0.6458 0.0894>]}
       GreenCurve {keys [<at 0 0 RCIM_Cubic>,<at 0.1492 0.0104>,<at 0.4863 0.2995>,<at 0.6458 0.0894>]}
       RedCurve {keys [<at 0 0.4323 RCIM_Cubic>,<at 0.1492 0.9992>,<at 0.4863 0.9274>,<at 0.6458 0.993>]}
      }})
    R2 DefaultColor:LinearColor=&ColorFromCurve(ColorCurve=DataInterface() {propertyValues {
       AlphaCurve {keys [<at 0 1>,<at 0.0914 0.2857>,<at 0.308 0.05>,<at 0.5592 0.2>,<at 0.8632 0>]}
       BlueCurve {keys [<at 0.0009 0>,<at 0.2668 -0.2829>,<at 0.5114 0.8849>,<at 0.619 1>,<at 0.682 0.383>]}
       GreenCurve {keys [<at 0.0009 0.0767>,<at 0.2742 -0.1994>,<at 0.5114 0.8849>,<at 0.619 1>,<at 0.682 0.2585>]}
       RedCurve {keys [<at 0.0009 3.2194>,<at 0.0016 3.2194>,<at 0.4479 0.8849>,<at 0.5114 0.8849>,<at 0.619 1>]}
      }})
    R2 ForestGreen:LinearColor=&ColorFromCurve(ColorCurve=DataInterface() {propertyValues {
       AlphaCurve {keys [<at 0 6>,<at 0.1907 1>,<at 0.3261 0.05>,<at 0.5495 0.2>,<at 0.7007 0.2>,<at 0.8474 0 RCIM_Cubic>]}
       BlueCurve {keys [<at 0 0.0172 RCIM_Cubic>,<at 0.2179 0.0322>,<at 0.4863 0.4656>,<at 0.6537 0.1027>]}
       GreenCurve {keys [<at 0 0.3438 RCIM_Cubic>,<at 0.2179 0.245>,<at 0.4863 0.875>,<at 0.6537 0.2987>]}
       RedCurve {keys [<at 0 0 RCIM_Cubic>,<at 0.2179 0.021>,<at 0.4863 0.4466>,<at 0.6537 0.0924>]}
      }})
    R2 Grey:LinearColor=&ColorFromCurve(ColorCurve=DataInterface() {propertyValues {
       AlphaCurve {keys [<at 0 4>,<at 0.1907 1>,<at 0.3261 0.05>,<at 0.5495 0.2>,<at 0.7007 0.2>,<at 0.8474 0 RCIM_Cubic>]}
       BlueCurve {keys [<at 0 0.0692 RCIM_Cubic>,<at 0.2285 0.1336>,<at 0.4863 0.9274>,<at 0.6405 0.132>]}
       GreenCurve {keys [<at 0 0.0781 RCIM_Cubic>,<at 0.2285 0.1336>,<at 0.4863 0.9274>,<at 0.6405 0.132>]}
       RedCurve {keys [<at 0 0.0767 RCIM_Cubic>,<at 0.2285 0.1406>,<at 0.4863 0.9274>,<at 0.6405 0.1354>]}
      }})
    R2 Lime:LinearColor=&ColorFromCurve(ColorCurve=DataInterface() {propertyValues {
       AlphaCurve {keys [<at 0 6>,<at 0.1907 1>,<at 0.3261 0.05>,<at 0.5495 0.2>,<at 0.8122 0 RCIM_Cubic>]}
       BlueCurve {keys [<at 0 0 RCIM_Cubic>,<at 0.2179 0.0459>,<at 0.4863 0.9274>,<at 0.6299 0.0724>]}
       GreenCurve {keys [<at 0 0.6875 RCIM_Cubic>,<at 0.2179 0.4635>,<at 0.4863 0.9274>,<at 0.6299 0.7708>]}
       RedCurve {keys [<at 0 0.3447 RCIM_Cubic>,<at 0.2179 0.2553>,<at 0.4863 0.9274>,<at 0.6299 0.4226>]}
      }})
    R2 Orange:LinearColor=&ColorFromCurve(ColorCurve=DataInterface() {propertyValues {
       AlphaCurve {keys [<at 0 7>,<at 0.1907 1>,<at 0.3261 0.05>,<at 0.5495 0.2>,<at 0.7007 0.2>,<at 0.8474 0 RCIM_Cubic>]}
       BlueCurve {keys [<at 0 0 RCIM_Cubic>,<at 0.2364 0.0598>,<at 0.4863 0.4263>,<at 0.6217 0.1191>]}
       GreenCurve {keys [<at 0 0.2834 RCIM_Cubic>,<at 0.2364 0.2792>,<at 0.4863 0.5719>,<at 0.6217 0.3121>]}
       RedCurve {keys [<at 0 0.8125 RCIM_Cubic>,<at 0.2364 0.6889>,<at 0.4863 0.8438>,<at 0.6217 0.6724>]}
      }})
    R2 Pink:LinearColor=&ColorFromCurve(ColorCurve=DataInterface() {propertyValues {
       AlphaCurve {keys [<at 0 7>,<at 0.1907 1>,<at 0.3261 0.05>,<at 0.5495 0.2>,<at 0.7007 0.2>,<at 0.8474 0 RCIM_Cubic>]}
       BlueCurve {keys [<at 0 0.3419 RCIM_Cubic>,<at 0.2707 0.3564>,<at 0.4863 0.7216>,<at 0.6141 0.5634>]}
       GreenCurve {keys [<at 0 0.0612 RCIM_Cubic>,<at 0.2707 0.1489>,<at 0.4863 0.5789>,<at 0.6141 0.2536>]}
       RedCurve {keys [<at 0 0.5029 RCIM_Cubic>,<at 0.2707 0.5614>,<at 0.4863 0.855>,<at 0.6141 0.8542>]}
      }})
    R2 Purple:LinearColor=&ColorFromCurve(ColorCurve=DataInterface() {propertyValues {
       AlphaCurve {keys [<at 0 7>,<at 0.1907 1>,<at 0.3261 0.05>,<at 0.5495 0.2>,<at 0.7007 0.2>,<at 0.8474 0 RCIM_Cubic>]}
       BlueCurve {keys [<at 0 0.8281 RCIM_Cubic>,<at 0.2919 0.6836>,<at 0.4863 0.8957>,<at 0.6141 0.9565>]}
       GreenCurve {keys [<at 0 0 RCIM_Cubic>,<at 0.2919 0.0563>,<at 0.4863 0.2791>,<at 0.6141 0.1097>]}
       RedCurve {keys [<at 0 0.1924 RCIM_Cubic>,<at 0.2919 0.3349>,<at 0.4863 0.5023>,<at 0.6141 0.3436>]}
      }})
    R2 Saffron:LinearColor=&ColorFromCurve(ColorCurve=DataInterface() {propertyValues {
       AlphaCurve {keys [<at 0 4>,<at 0.1907 1>,<at 0.3261 0.05>,<at 0.5495 0.2>,<at 0.8175 0 RCIM_Cubic>]}
       BlueCurve {keys [<at 0 0 RCIM_Cubic>,<at 0.2697 0.0046>,<at 0.4863 0.3188>,<at 0.6748 0.0798>]}
       GreenCurve {keys [<at 0 0.6105 RCIM_Cubic>,<at 0.2697 0.6881>,<at 0.4863 0.9274>,<at 0.6748 0.6052>]}
       RedCurve {keys [<at 0 0.7969 RCIM_Cubic>,<at 0.2697 0.8967>,<at 0.4863 0.8744>,<at 0.6748 0.7656>]}
      }})
    R2 SkyBlue:LinearColor=&ColorFromCurve(ColorCurve=DataInterface() {propertyValues {
       AlphaCurve {keys [<at 0 7>,<at 0.1907 1>,<at 0.3261 0.05>,<at 0.5495 0.2>,<at 0.8096 0 RCIM_Cubic>]}
       BlueCurve {keys [<at 0 0.5333 RCIM_Cubic>,<at 0.2496 0.5525>,<at 0.4863 0.9323>,<at 0.6141 0.9323>]}
       GreenCurve {keys [<at 0 0.3564 RCIM_Cubic>,<at 0.2496 0.3792>,<at 0.4863 0.8083>,<at 0.6141 0.6988>]}
       RedCurve {keys [<at 0 0 RCIM_Cubic>,<at 0.2496 0.0875>,<at 0.4863 0.5584>,<at 0.6141 0.2282>]}
      }})
    R2 TitaniumWhite:LinearColor=&ColorFromCurve(ColorCurve=DataInterface() {propertyValues {
       AlphaCurve {keys [<at 0 4>,<at 0.1907 1>,<at 0.3261 0.05>,<at 0.5495 0.2>,<at 0.7007 0.2>,<at 0.8175 0 RCIM_Cubic>]}
       BlueCurve {keys [<at 0 0.1846 RCIM_Cubic>,<at 0.2285 0.1336>,<at 0.4863 0.9274>,<at 0.6405 0.3248>]}
       GreenCurve {keys [<at 0 0.2083 RCIM_Cubic>,<at 0.2285 0.1336>,<at 0.4863 0.9274>,<at 0.6405 0.3248>]}
       RedCurve {keys [<at 0 0.2046 RCIM_Cubic>,<at 0.2285 0.1406>,<at 0.4863 0.9274>,<at 0.6405 0.3333>]}
      }})
    R3 Black:LinearColor=<rgba 0.0497 0.0497 0.0497 1>
    R3 BurntSienna:LinearColor=<rgba 0.3 0.0926 0.037 1>
    R3 Cobalt:LinearColor=<rgba 0.0904 0.0923 1.0208 1>
    R3 Crimson:LinearColor=<rgba 4 0 0.1084 1>
    R3 DefaultColor:LinearColor=<rgba 1 1 1 0.3>
    R3 ForestGreen:LinearColor=<rgba 0.2865 1 0.3225 1>
    R3 Grey:LinearColor=<rgba 0.135 0.135 0.135 1>
    R3 Lime:LinearColor=<rgba 0.2422 0.4844 0 0.6>
    R3 Orange:LinearColor=<rgba 0.6174 0.2247 0.0643 0.8>
    R3 Pink:LinearColor=<rgba 1 0.25 0.7273 1>
    R3 Purple:LinearColor=<rgba 0.1736 0.0879 0.9375 1>
    R3 Saffron:LinearColor=<rgba 0.5938 0.474 0.0804 0.3>
    R3 SkyBlue:LinearColor=<rgba 0.1062 0.3925 0.5365 1>
    R4 BurntSienna:LinearColor=<rgba 0.0312 0.0096 0.0039 0.6>
    R4 Cobalt:LinearColor=<rgba 0 0.0115 0.6198 0.7>
    R4 Crimson:LinearColor=<rgba 0.2656 0 0.0072 0.7>
    R4 DefaultColor:LinearColor=<rgba 0.0356 0.0369 0.9047 0.5>
    R4 ForestGreen:LinearColor=<rgba 0 0.2552 0.0129 0.6>
    R4 Grey:LinearColor=<rgba 0.0417 0.0417 0.0417 0.7>
    R4 Lime:LinearColor=<rgba 0.1875 0.375 0 0.6>
    R4 Orange:LinearColor=<rgba 0.5521 0.1706 0 0.45>
    R4 Pink:LinearColor=<rgba 0.3906 0.0468 0.2656 0.5>
    R4 Purple:LinearColor=<rgba 0.1567 0 0.7552 0.5>
    R4 Saffron:LinearColor=<rgba 0.5573 0.4299 0 0.3>
    R4 SkyBlue:LinearColor=<rgba 0.0137 0.2957 0.4375 0.5>
    R4 TitaniumWhite:LinearColor=<rgba 0.224 0.224 0.224 0.6>
    R5 Black:LinearColor=<rgba 0.0469 0.0469 0.0469 0.6>
    R5 BurntSienna:LinearColor=<rgba 0.1042 0.0322 0.0128 0.7>
    R5 Cobalt:LinearColor=<rgba 0.0312 0.0401 1 0.6>
    R5 Crimson:LinearColor=<rgba 0.4635 0.0266 0.0384 0.6>
    R5 DefaultColor:LinearColor=<rgba 1 1 1 0.5>
    R5 ForestGreen:LinearColor=<rgba 0.0704 0.901 0.1124 0.4>
    R5 Grey:LinearColor=<rgba 0.1667 0.1667 0.1667 0.4>
    R5 Lime:LinearColor=<rgba 0.2938 0.5104 0.0771 0.25>
    R5 Orange:LinearColor=<rgba 0.8261 0.2993 0.0172 0.3>
    R5 Pink:LinearColor=<rgba 0.651 0.1763 0.4784 0.4>
    R5 Purple:LinearColor=<rgba 0.2704 0.1155 0.7917 0.3>
    R5 Saffron:LinearColor=<rgba 0.8906 0.6824 0 0.15>
    R5 SkyBlue:LinearColor=<rgba 0.0754 0.3822 0.5365 0.4>
    R5 TitaniumWhite:LinearColor=<rgba 0.5729 0.5729 0.5729 0.3>
    Use R1 Painted Overrides=true
    Use R2 Painted Overrides=true
    Use R3 Painted Overrides=true
    Use R4 Painted Overrides=true
    Use R5 Painted Overrides=true
   ) {module=VC_PaintedVariantsOverride001}
   &ParticleState()
   &InheritSourceMovement(ApplyForce=false, ApplyPosition=true, ApplyPositionFalloff=false)
   &SolveForcesAndVelocity(Clamp Velocity=false, Limit Acceleration=false)
   &VC_UpdateRibbon_v1(Enable Skeletal Position=true)
   /SetVariables(
    Particles.R1Position:NiagaraPosition=&AddVectorToPosition(
     Position=$Particles.Position
     Vector=/VehicleCosmetics/FX/Niagara/Modules/VC_ConvertVectorToWorld(Vector:Vector3f=&MakeVector(
       X=&FloatFromCurve(
        FloatCurve=DataInterface() {propertyValues {Curve {defaultValue=1, keys [{tangentMode=RCTM_User, time=0.45, value=0}, {tangentMode=RCTM_User, time=0.6219, value=0}]}}}
        Scale Curve=0
       )
       Y=&Multiply_Float(
        A=&Cosine(Normalized Angle=$Emitter.Age, Period=0.5, Scale=-5, Bias=1)
        B=&FloatFromCurve(
         FloatCurve=DataInterface() {propertyValues {Curve {
            defaultValue=1
            keys [{interpMode=RCIM_Cubic, tangentMode=RCTM_Break, time=0.5, value=0}, {interpMode=RCIM_Cubic, tangentMode=RCTM_Break, time=1, value=1}]
           }}}
         Scale Curve=0
        )
       )
       Z=&FloatFromCurve(
        FloatCurve=DataInterface() {propertyValues {Curve {
           defaultValue=1
           keys [
            {interpMode=RCIM_Cubic, tangentMode=RCTM_Break, time=0.1117, value=0.7018}
            {interpMode=RCIM_Cubic, tangentMode=RCTM_Break, time=1, value=0.9}
           ]
          }}}
        Scale Curve=0
       )
      ))
    )
    Particles.R1_Width:NiagaraFloat=&FloatFromCurve(
     FloatCurve=DataInterface() {propertyValues {Curve {keys [
         {arriveTangent=0.6775, interpMode=RCIM_Cubic, leaveTangent=0.6775, tangentMode=RCTM_User, time=0.095, value=1.0904}
         {arriveTangent=-0.332, interpMode=RCIM_Cubic, leaveTangent=-0.332, time=0.1964, value=1.2425}
         {arriveTangent=-2.6557, interpMode=RCIM_Cubic, leaveTangent=-2.6557, time=0.3731, value=0.9981}
         {arriveTangent=-0.4746, interpMode=RCIM_Cubic, leaveTangent=-0.4746, time=0.4941, value=0.4518}
         {arriveTangent=2.5628, interpMode=RCIM_Cubic, leaveTangent=2.5628, tangentMode=RCTM_User, time=0.6606, value=0.8616}
         {arriveTangent=0.7777, interpMode=RCIM_Cubic, leaveTangent=1.1948, tangentMode=RCTM_Break, time=0.8603, value=0.8986}
        ]}}}
     Scale Curve=90.1385
    )
    Particles.R2 Position:NiagaraPosition=&AddVectorToPosition(
     Position=$Particles.Position
     Vector=&VC_ConvertVectorToWorld(Vector:Vector3f=&MakeVector(
       X=&FloatFromCurve(
        FloatCurve=DataInterface() {propertyValues {Curve {defaultValue=1, keys [{tangentMode=RCTM_User, time=0.45, value=0}, {tangentMode=RCTM_User, time=0.6219, value=1.6}]}}}
        Scale Curve=32
       )
       Y=&Multiply_Float(
        A=&Cosine(Normalized Angle=$Emitter.Age, Period=0.5, Scale=-5, Bias=1)
        B=&FloatFromCurve(
         FloatCurve=DataInterface() {propertyValues {Curve {
            defaultValue=1
            keys [{interpMode=RCIM_Cubic, tangentMode=RCTM_Break, time=0.5, value=0}, {interpMode=RCIM_Cubic, tangentMode=RCTM_Break, time=1, value=1}]
           }}}
         Scale Curve=0
        )
       )
       Z=&FloatFromCurve(
        FloatCurve=DataInterface() {propertyValues {Curve {
           defaultValue=1
           keys [
            {interpMode=RCIM_Cubic, tangentMode=RCTM_Break, time=0.1117, value=0.7018}
            {interpMode=RCIM_Cubic, tangentMode=RCTM_Break, time=1, value=0.9}
           ]
          }}}
        Scale Curve=-30.4006
       )
      ))
    )
    Particles.R2_Width:NiagaraFloat=&FloatFromCurve(
     FloatCurve=DataInterface() {propertyValues {Curve {keys [
         {arriveTangent=1.3035, interpMode=RCIM_Cubic, time=0.1211, value=0.4903}
         {arriveTangent=0.2508, interpMode=RCIM_Cubic, leaveTangent=0.2508, time=0.1738, value=0.5077}
         {interpMode=RCIM_Cubic, tangentMode=RCTM_Break, time=0.3314, value=0.5431}
         {arriveTangent=0.0109, interpMode=RCIM_Cubic, leaveTangent=0.0109, time=0.4446, value=0.2768}
         <at 0.7986 0.5482 RCIM_Cubic>
        ]}}}
     Scale Curve=90.893
    )
    Particles.R3_Position:NiagaraPosition=&AddVectorToPosition(
     Position=$Particles.Position
     Vector=&VC_ConvertVectorToWorld(Vector:Vector3f=&MakeVector(
       X=&FloatFromCurve(
        FloatCurve=DataInterface() {propertyValues {Curve {defaultValue=1, keys [{tangentMode=RCTM_User, time=0.447, value=1.26}, {tangentMode=RCTM_User, time=0.62, value=1.6}]}}}
        Scale Curve=20
       )
       Y=&Multiply_Float(
        A=&Cosine(Normalized Angle=$Emitter.Age, Period=0.5, Scale=-5, Bias=1)
        B=&FloatFromCurve(
         FloatCurve=DataInterface() {propertyValues {Curve {
            defaultValue=1
            keys [{interpMode=RCIM_Cubic, tangentMode=RCTM_Break, time=0.5, value=0}, {interpMode=RCIM_Cubic, tangentMode=RCTM_Break, time=1, value=1}]
           }}}
         Scale Curve=0
        )
       )
       Z=&FloatFromCurve(
        FloatCurve=DataInterface() {propertyValues {Curve {
           defaultValue=1
           keys [
            {interpMode=RCIM_Cubic, tangentMode=RCTM_Break, time=-0.0151, value=0.3961}
            {interpMode=RCIM_Cubic, tangentMode=RCTM_Break, time=5.7612, value=9.1732}
           ]
          }}}
        Scale Curve=0
       )
      ))
    )
    Particles.R3_Width:NiagaraFloat=&FloatFromCurve(
     FloatCurve=DataInterface() {propertyValues {Curve {keys [
         {interpMode=RCIM_Cubic, tangentMode=RCTM_Break, time=0.0512, value=0}
         {arriveTangent=3.1865, interpMode=RCIM_Cubic, leaveTangent=3.1865, time=0.2001, value=1.0144}
         {arriveTangent=-1.0643, interpMode=RCIM_Cubic, leaveTangent=-1.8754, tangentMode=RCTM_Break, time=0.365, value=1}
         <at 0.6835 0.4 RCIM_Cubic>
        ]}}}
     Scale Curve=85
    )
    Particles.R4Width:NiagaraFloat=&FloatFromCurve(
     FloatCurve=DataInterface() {propertyValues {Curve {keys [
         {arriveTangent=0.2294, interpMode=RCIM_Cubic, leaveTangent=0.1215, tangentMode=RCTM_Break, time=-0.1741, value=0.8351}
         {arriveTangent=0.1518, interpMode=RCIM_Cubic, leaveTangent=0.1518, tangentMode=RCTM_User, time=0.3234, value=1.1317}
         <at 0.6882 0.63 RCIM_Cubic>
        ]}}}
     Scale Curve=60
    )
    Particles.R4_Position:NiagaraPosition=&AddVectorToPosition(
     Position=$Particles.Position
     Vector=&VC_ConvertVectorToWorld(Vector:Vector3f=&MakeVector(
       X=&FloatFromCurve(
        FloatCurve=DataInterface() {propertyValues {Curve {defaultValue=1, keys [{tangentMode=RCTM_User, time=0.4442, value=0.0272}, {tangentMode=RCTM_User, time=0.6219, value=1.6}]}}}
        Scale Curve=0
       )
       Y=&Multiply_Float(
        A=&Cosine(Normalized Angle=$Emitter.Age, Period=0.5, Scale=-5, Bias=1)
        B=&FloatFromCurve(
         FloatCurve=DataInterface() {propertyValues {Curve {
            defaultValue=1
            keys [{interpMode=RCIM_Cubic, tangentMode=RCTM_Break, time=0.5, value=0}, {interpMode=RCIM_Cubic, tangentMode=RCTM_Break, time=1, value=1}]
           }}}
         Scale Curve=0
        )
       )
       Z=&FloatFromCurve(
        FloatCurve=DataInterface() {propertyValues {Curve {
           defaultValue=1
           keys [
            {interpMode=RCIM_Cubic, tangentMode=RCTM_Break, time=-0.0151, value=0.3961}
            {interpMode=RCIM_Cubic, tangentMode=RCTM_Break, time=5.7612, value=9.1732}
           ]
          }}}
        Scale Curve=39.9285
       )
      ))
    )
   ) {isSetParameters=true}
   /SetVariables(
    Particles.R1_Color_Scaled:LinearColor=&MultiplyLinearColor(
     Linear Color=$Particles.R1 Color
     Scale Factor=&MakeVector4FromVectorAndFloat(
      "Vector (RGB)"=&VectorFromFloat(Value=&FloatFromCurve(
        FloatCurve=DataInterface() {propertyValues {Curve {keys [
            <at -0.1265 12.32 RCIM_Cubic>
            {arriveTangent=-21.3105, interpMode=RCIM_Cubic, leaveTangent=-21.3105, time=0.1072, value=17.8443}
            {arriveTangent=-3.2592, interpMode=RCIM_Cubic, leaveTangent=-3.2592, tangentMode=RCTM_User, time=0.3385, value=2.4113}
            {arriveTangent=2.5648, interpMode=RCIM_Cubic, leaveTangent=2.5648, tangentMode=RCTM_User, time=0.64, value=1.973}
            {arriveTangent=22.1779, interpMode=RCIM_Cubic, leaveTangent=22.1779, time=1.0714, value=20.2083}
            {arriveTangent=-34.897, interpMode=RCIM_Cubic, leaveTangent=-34.897, time=1.4558, value=20.0672}
            {arriveTangent=-41.0956, interpMode=RCIM_Cubic, leaveTangent=-41.0956, time=1.5789, value=2.4972}
            <at 1.7789 6.7918 RCIM_Cubic>
           ]}}}
        Scale Curve=&MakeCustomFloatFromBool(Bool=$User.bIsPainted, True Float=$User.Power_EmissivePaintedR4, False Float=2)
       ))
      "Float (Alpha)"=&FloatFromCurve(
       FloatCurve=DataInterface() {propertyValues {Curve {keys [
           <at -0.0449 3.0504 RCIM_Cubic>
           {arriveTangent=0.1754, interpMode=RCIM_Cubic, leaveTangent=0.1754, time=0.111, value=0.7392}
           {arriveTangent=-0.2383, interpMode=RCIM_Cubic, leaveTangent=-0.2383, tangentMode=RCTM_User, time=0.4793, value=3.1423}
           {arriveTangent=-43.3151, interpMode=RCIM_Cubic, leaveTangent=-30.5159, tangentMode=RCTM_Break, time=0.675, value=3.1272}
           {arriveTangent=-9.9909, interpMode=RCIM_Cubic, leaveTangent=-9.9909, time=0.8329, value=0.2261}
           <at 1.01 -0.2199 RCIM_Cubic>
          ]}}}
       Scale Curve=&MakeCustomFloatFromBool(Bool=$User.bIsPainted, True Float=$User.Power_AlphaPaintedR4, False Float=0.5)
      )
     )
    )
    Particles.R3_Color_Scaled:LinearColor=&MultiplyLinearColor(
     Linear Color=$Particles.R3 Color
     Scale Factor=&Vector4FromCurve(Vector4Curve=DataInterface() {propertyValues {
        AlphaCurve {keys [<at 0 0.8 RCIM_Cubic>,<at 0.2711 0.45>,<at 0.4802 0.4>,<at 0.6793 0.363>,<at 0.985 0>]}
        BlueCurve {keys [<at 0 0.8675>]}
        GreenCurve {keys [<at 0 0.8675>]}
        RedCurve {keys [<at 0 0.8675>]}
       }})
    )
    Particles.R4_Color_Scaled:LinearColor=&MultiplyLinearColor(
     Linear Color=$Particles.R4 Color
     Scale Factor=&MakeVector4FromVectorAndFloat(
      "Vector (RGB)"=&VectorFromFloat(Value=&FloatFromCurve(
        FloatCurve=DataInterface() {propertyValues {Curve {keys [
            <at -0.0135 1.393 RCIM_Cubic>
            {arriveTangent=0.2114, interpMode=RCIM_Cubic, leaveTangent=0.2114, time=0.1169, value=2.0317}
            {arriveTangent=-2.6168, interpMode=RCIM_Cubic, leaveTangent=-2.6168, tangentMode=RCTM_User, time=0.887, value=1.5834}
            {interpMode=RCIM_Cubic, leaveTangent=-0.8868, time=1.0581, value=1.0126}
           ]}}}
        Scale Curve=&MakeCustomFloatFromBool(Bool=$User.bIsPainted, True Float=$User.Power_EmissivePaintedR4, False Float=0.6)
       ))
      "Float (Alpha)"=&FloatFromCurve(
       FloatCurve=DataInterface() {propertyValues {Curve {keys [
           <at -0.0014 1.7739 RCIM_Cubic>
           {arriveTangent=-1.9854, interpMode=RCIM_Cubic, leaveTangent=-1.9854, time=0.3044, value=2.258}
           {arriveTangent=-0.2383, interpMode=RCIM_Cubic, leaveTangent=-1.8935, tangentMode=RCTM_Break, time=0.7964, value=0.1899}
           <at 1.4148 -1.634 RCIM_Cubic>
          ]}}}
       Scale Curve=&MakeCustomFloatFromBool(Bool=$User.bIsPainted, True Float=$User.Power_AlphaPaintedR4, False Float=1.2)
      )
     )
    )
    Particles.R5_Color_Scaled:LinearColor=&MultiplyLinearColor(
     Linear Color=$Particles.R5 Color
     Scale Factor=&MakeVector4FromVectorAndFloat(
      "Vector (RGB)"=&VectorFromFloat(Value=&FloatFromCurve(
        FloatCurve=DataInterface() {propertyValues {Curve {keys [
            <at -0.17 16.4711 RCIM_Cubic>
            {arriveTangent=-27.6507, interpMode=RCIM_Cubic, leaveTangent=-27.6507, time=0.1072, value=17.8443}
            {arriveTangent=-3.2592, interpMode=RCIM_Cubic, leaveTangent=-3.2592, tangentMode=RCTM_User, time=0.3385, value=2.4113}
            {arriveTangent=2.5648, interpMode=RCIM_Cubic, leaveTangent=2.5648, tangentMode=RCTM_User, time=0.64, value=1.973}
            {arriveTangent=11.7336, interpMode=RCIM_Cubic, leaveTangent=11.7336, time=1.0714, value=20.2083}
            {arriveTangent=-34.897, interpMode=RCIM_Cubic, leaveTangent=-34.897, time=1.3482, value=10.2825}
            {arriveTangent=-8.1043, interpMode=RCIM_Cubic, leaveTangent=-8.1043, time=1.5789, value=2.4972}
            <at 1.7789 6.7918 RCIM_Cubic>
           ]}}}
        Scale Curve=&MakeCustomFloatFromBool(Bool=$User.bIsPainted, True Float=$User.Power_EmissivePaintedR4, False Float=0.7)
       ))
      "Float (Alpha)"=&FloatFromCurve(
       FloatCurve=DataInterface() {propertyValues {Curve {keys [
           <at -0.0556 3.887 RCIM_Cubic>
           {arriveTangent=-1.3925, interpMode=RCIM_Cubic, leaveTangent=-1.3925, time=0.111, value=0.7392}
           {arriveTangent=-0.2383, interpMode=RCIM_Cubic, leaveTangent=-0.2383, tangentMode=RCTM_User, time=0.4793, value=3.1423}
           {arriveTangent=-6.3349, interpMode=RCIM_Cubic, leaveTangent=-6.3349, time=0.8329, value=0.2261}
           <at 1.01 -0.2199 RCIM_Cubic>
          ]}}}
       Scale Curve=&MakeCustomFloatFromBool(Bool=$User.bIsPainted, True Float=$User.Power_AlphaPaintedR4, False Float=0.3)
      )
     )
    )
   ) {isSetParameters=true}
  ]
  rendererClasses [&NiagaraRibbonRendererProperties]
  renderers [
   {rendererClass=&NiagaraRibbonRendererProperties}
   {rendererClass=&NiagaraRibbonRendererProperties, rendererIndex=1}
   {rendererClass=&NiagaraRibbonRendererProperties, rendererIndex=2}
   {rendererClass=&NiagaraRibbonRendererProperties, rendererIndex=3}
   {rendererClass=&NiagaraRibbonRendererProperties, rendererIndex=4}
  ]
 }
 NE_VC_Boost_Sprite_Parent {
  emitterUpdateScript [
   &EmitterState()
   &VC_FilteredSocketCount()
   &SpawnRate(SpawnRate=60, Use Spawn Probability=false)
   &SpawnRate(SpawnGroup=2, SpawnRate=80, Use Spawn Probability=false) {module=SpawnRate002}
   &SpawnRate(SpawnGroup=1, SpawnRate=35, Use Spawn Probability=false) {module=SpawnRate001}
  ]
  particleSpawnScript [
   &VC_PaintedVariantsOverride(
    Black=&SelectLinearColorFromArray(
     Array Sampling Mode=::NewEnumerator0
     Direct Array Index=?
     Color Selection Array=DataInterface() {propertyValues {ColorData [<rgba 0.0573 0.0573 0.0573 1>, <rgba 0.0194 0.0194 0.0194 1>, <rgba 0 0 0 0.8>, <rgba 0.0497 0.0497 0.0497 1>]}}
     Random Seed=0
    )
    BurntSienna=<rgba 0.1771 0.0547 0.0218 1>
    Cobalt=<rgba 0.0156 0.0338 1 1>
    Crimson=<rgba 0.6979 0.0436 0.0614 1>
    DefaultColor=&SelectLinearColorFromArray(
     Array Sampling Mode=::NewEnumerator0
     Direct Array Index=?
     Color Selection Array=DataInterface() {propertyValues {ColorData [
        <rgba 0.0908 0.0423 0.7231 1>
        <rgba 0.807 0.1981 0.9823 1>
        <rgba 0.7991 0.1912 0.1845 0.8>
        <rgba 0.0595 0.7835 0.9301 1>
        <rgba 0.7454 0.4125 0.0578 0.8>
        <rgba 1 1 1 0>
       ]}}
     Random Seed=0
    )
    ForestGreen=<rgba 0.0964 0.5 0.1167 1>
    Lime=<rgba 0.3801 0.5885 0.1717 1>
    Orange=<rgba 0.6724 0.3303 0.1471 1>
    Pink=<rgba 0.5052 0.1895 0.3904 1>
    Purple=<rgba 0.1682 0.0701 0.8957 1>
    Saffron=<rgba 0.8333 0.7047 0.2821 1>
    SkyBlue=<rgba 0.1481 0.4065 0.5365 1>
    TitaniumWhite=<rgba 0.7604 0.7604 0.7604 0.8>
   )
   &InitializeParticle(
    Color Mode=::NewEnumerator0
    Lifetime Max=0.25
    Lifetime Min=0.15
    Lifetime Mode=::NewEnumerator1
    Position Mode=::NewEnumerator0
    Sprite Size Max=<1.7777 210.2738>
    Sprite Size Min=<0.755 132.6772>
    Sprite Size Mode=::NewEnumerator2
    Write Material Random=true
   )
   &PartitionParticles(Number Of Partitions=$Emitter.PortCount, Particles Per Partition=2, Partition Mode:ENiagara_PartitionMode=::NewEnumerator1)
   &SkeletalMeshLocation(
    Bone / Socket Index=$Particles.PartitionParticles.Partition
    Bone / Socket Sampling Mode=::NewEnumerator5
    MaskSpawnGroup=false
    Mesh Sampling Type=::NewEnumerator6
    Sampled Position Offset=&AddVectorToPosition(Position=$User.DTCarOffset, Vector=<-90 0 0>)
    Skeletal Mesh=$System.SkeletalMesh
   )
   &AddVelocity(Velocity=<-1500 0 0>)
   &ShapeLocation(
    MaskSpawnGroup=false
    Offset=<40 0 0>
    Offset Coordinate Space=::Local
    Offset Mode=::NewEnumerator0
    Override Seed=false
    Shape Origin=$Particles.Position
    Sphere Radius=&FloatFromCurve(
     FloatCurve=DataInterface() {propertyValues {Curve {
        defaultValue=1
        keys [{tangentMode=RCTM_User, time=-0.0112, value=1.1343}, <at 0.2579 1.6144>, {tangentMode=RCTM_User, time=0.7301, value=1.7851}]
       }}}
     Scale Curve=40
    )
    UseOwnerScale=true
   )
   /SetVariables(
    Particles.L1_Color:LinearColor=&ColorFromCurve(ColorCurve=DataInterface() {propertyValues {
       AlphaCurve {keys [<at 0 1 RCIM_Cubic>,<at 0.9172 0 RCIM_Cubic>]}
       BlueCurve {keys [<at 0 1 RCIM_Cubic>,<at 1 1 RCIM_Cubic>]}
       GreenCurve {keys [<at 0 1 RCIM_Cubic>,<at 1 1 RCIM_Cubic>]}
       RedCurve {keys [<at 0 1 RCIM_Cubic>,<at 1 1 RCIM_Cubic>]}
      }})
    Particles.NewRendererTag:NiagaraInt32=&SpawnGroup()
    Particles.T2_Color:LinearColor=<rgba 1 0 0 0.8>
    Particles.T3_Color:LinearColor=<rgba 0.6806 0.9292 0.0119 0.6>
    Particles.T4_Color:LinearColor=<rgba 0.0156 0.8795 0.7687 0.9>
    Particles.T_BasicSize:Vector2f=&RandomRangeVector2D(Minimum=<16.3547 6.6505>, Maximum=<67.0299 145>)
    Particles.Triangles2_Size:Vector2f=&RandomRangeVector2D(Minimum=<20.7099 14.5277>, Maximum=<45.7668 30.5539>)
   ) {isSetParameters=true, module=SetVariables_#2}
  ]
  particleUpdateScript [
   &ParticleState()
   &InheritSourceMovement(ApplyForce=false, ApplyPosition=true, ApplyPositionFalloff=false)
   &SolveForcesAndVelocity(Clamp Velocity=false, Limit Acceleration=false)
   &SubUVAnimation(
    End Frame Range Override=4
    Evaluation Type=::NewEnumerator0
    "Sprite Renderer "=DataInterface() {propertyValues {SpriteRenderer=&self:NE_VC_Boost_Sprite_Parent.NiagaraSpriteRendererProperties_2}}
    SubUV Animation Mode:ENiagara_SubUVLookupModeV2=::NewEnumerator2
    UseEndFrame=true
    UseStartFrame=true
   )
   /SetVariables(Particles.OutlineT_Rotation:NiagaraFloat=&RandomRangeFloat(Minimum=0.6636, Maximum=65.2032)) {isSetParameters=true, module=SetVariables_#3}
  ]
  rendererClasses [&NiagaraSpriteRendererProperties]
  renderers [
   {rendererClass=&NiagaraSpriteRendererProperties}
   {rendererClass=&NiagaraSpriteRendererProperties, rendererIndex=1}
   {rendererClass=&NiagaraSpriteRendererProperties, rendererIndex=2}
   {rendererClass=&NiagaraSpriteRendererProperties, rendererIndex=3}
   {rendererClass=&NiagaraSpriteRendererProperties, rendererIndex=4}
  ]
 }
 NE_VC_Boost_Sprite_Parent002 {
  emitterUpdateScript [
   &EmitterState()
   &VC_FilteredSocketCount()
   &SpawnRate(SpawnRate=85, Use Spawn Probability=false)
   &SpawnRate(SpawnGroup=1, SpawnRate=70, Use Spawn Probability=false) {module=SpawnRate001}
  ]
  particleSpawnScript [
   &VC_PaintedVariantsOverride(
    Black=<rgba 0.0156 0.0152 0.0139 0.8>
    BurntSienna=<rgba 0.6302 0.1721 0.0492 0.5>
    Cobalt=<rgba 0.0781 0.0952 1 0.8>
    Crimson=<rgba 4 0.2083 0.3111 0.7>
    DefaultColor=&ColorFromCurve(ColorCurve=DataInterface() {propertyValues {
       AlphaCurve {keys [<at 0 1 RCIM_Cubic>,<at 0.7814 0.7868>,<at 0.9846 0 RCIM_Cubic>]}
       BlueCurve {keys [<at 0 1 RCIM_Cubic>,<at 1 1 RCIM_Cubic>]}
       GreenCurve {keys [<at 0 1 RCIM_Cubic>,<at 1 1 RCIM_Cubic>]}
       RedCurve {keys [<at 0 1 RCIM_Cubic>,<at 1 1 RCIM_Cubic>]}
      }})
    ForestGreen=<rgba 0.4688 6 0.7482 0.6>
    Grey=<rgba 0.1042 0.1042 0.1042 1>
    Lime=<rgba 3.4062 6 0.8125 0.6>
    Orange=<rgba 6 2.2963 0.3125 0.6>
    Pink=<rgba 6 4.0938 5.3068 0.6>
    Purple=<rgba 1.6974 0.1816 3.875 0.7>
    Saffron=<rgba 2 1.6111 0.3333 1>
    SkyBlue=<rgba 0.3646 3.4489 5 1>
    TitaniumWhite=<rgba 6 6 6 0.6>
   )
   &InitializeParticle(
    Color Mode=::NewEnumerator0
    Lifetime Max=0.25
    Lifetime Min=0.12
    Lifetime Mode=::NewEnumerator1
    Position Mode=::NewEnumerator0
    Sprite Rotation Mode=::NewEnumerator2
    Sprite Size Max=<12.8989 315.693>
    Sprite Size Min=<5.5367 50>
    Sprite Size Mode=::NewEnumerator2
    Write Material Random=true
   )
   &PartitionParticles(Number Of Partitions=$Emitter.PortCount, Particles Per Partition=2, Partition Mode:ENiagara_PartitionMode=::NewEnumerator1)
   &SkeletalMeshLocation(
    Bone / Socket Index=$Particles.PartitionParticles.Partition
    Bone / Socket Sampling Mode=::NewEnumerator5
    MaskSpawnGroup=false
    Mesh Sampling Type=::NewEnumerator6
    Sampled Position Offset=&AddVectorToPosition(Position=$User.DTCarOffset, Vector=<-100 0 0>)
    Skeletal Mesh=$System.SkeletalMesh
   )
   &AddVelocity(
    Cone Angle=111.2954
    Module.UseConeFalloff=false
    Module.UseVelDistribution=false
    Override Seed=false
    Velocity Mode=::NewEnumerator2
    Velocity Speed=&RandomRangeFloat(Minimum=-328.368, Maximum=-598.6586)
   )
   &AddVelocity(Velocity=&MakeCustomVectorFromBool(Bool=$User.bIsFrontendPreview, True Vector=<-264.2505 0 1>, False Vector=<-203.5854 0 0>)) {module=AddVelocity001}
   &ShapeLocation(
    MaskSpawnGroup=false
    Offset=<70 0 0>
    Offset Coordinate Space=::Local
    Offset Mode=::NewEnumerator0
    Override Seed=false
    Shape Origin=$Particles.Position
    Sphere Radius=&RandomRangeFloat(Minimum=35, Maximum=45)
    UseOwnerScale=true
   )
   /SetVariables(
    Particles.CirclesRendererTag:NiagaraInt32=&SpawnGroup()
    Particles.DotsCR_Size:Vector2f=&RandomRangeVector2D(Minimum=<11.48 3.6027>, Maximum=<17.6672 11.7894>)
    Particles.T2_Color:LinearColor=<rgba 1 0 0 0.8>
    Particles.T3_Color:LinearColor=<rgba 0.6806 0.9292 0.0119 0.7>
    Particles.T4_Color:LinearColor=<rgba 0.0156 0.8795 0.7687 0.6>
    Particles.T_Outline_Size:Vector2f=&RandomRangeVector2D(Minimum=<7.8288 9.1509>, Maximum=<26.6119 54.5929>)
   ) {isSetParameters=true, module=SetVariables_#2}
  ]
  particleUpdateScript [
   &ParticleState()
   &InheritSourceMovement(ApplyForce=false, ApplyPosition=true, ApplyPositionFalloff=false)
   &SolveForcesAndVelocity(Clamp Velocity=false, Limit Acceleration=false)
   &SubUVAnimation(
    End Frame Range Override=4
    Evaluation Type=::NewEnumerator0
    "Sprite Renderer "=DataInterface() {propertyValues {SpriteRenderer=&self:NE_VC_Boost_Sprite_Parent002_0.NiagaraSpriteRendererProperties_0}}
    SubUV Animation Mode:ENiagara_SubUVLookupModeV2=::NewEnumerator2
    UseEndFrame=true
    UseStartFrame=true
   )
   /SetVariables(Particles.T_Rotation:NiagaraFloat=&RandomRangeFloat(Minimum=0.6636, Maximum=65.2032)) {isSetParameters=true, module=SetVariables_#3}
  ]
  rendererClasses [&NiagaraSpriteRendererProperties]
  renderers [
   {rendererClass=&NiagaraSpriteRendererProperties}
   {rendererClass=&NiagaraSpriteRendererProperties, rendererIndex=1}
   {rendererClass=&NiagaraSpriteRendererProperties, rendererIndex=2}
   {rendererClass=&NiagaraSpriteRendererProperties, rendererIndex=3}
  ]
 }
}
summary {
 systemName=NS_DyePacket
 emitters [
  {emitterName=NE_VC_Boost_Sprite_Parent, rendererClasses [&NiagaraSpriteRendererProperties]}
  {emitterName=NE_VC_Boost_Sprite_Parent002, rendererClasses [&NiagaraSpriteRendererProperties]}
  {emitterName=NE_MainRibbon_Boost, rendererClasses [&NiagaraRibbonRendererProperties]}
 ]
 userVariables [
  User.bIsFrontendPreview {type=NiagaraBool}
  User.bIsPainted {type=NiagaraBool}
  User.ColorVariantID {type=NiagaraInt32}
  User.CoreColor {type=LinearColor, description="", defaultValue {struct=LinearColor, value=<rgba 1 1 1 1>}}
  User.DTCarOffset {type=NiagaraPosition, description="", defaultValue {}}
  User.DTCarOffset001 {type=Vector3f, description="", defaultValue {struct=Vector3f, value=<0 0 0>}}
  User.DTCarOffset002 {type=Vector3f, description="", defaultValue {struct=Vector3f, value=<0 0 0>}}
  User.Emissive {type=NiagaraFloat, defaultValue=10, description=""}
  User.GlassColor {type=LinearColor, description="", defaultValue {struct=LinearColor, value=<rgba 1 1 1 1>}}
  User.GlassColorVar {type=NiagaraFloat, defaultValue=1, description=""}
  User.GlowColor {type=LinearColor, description="", defaultValue {struct=LinearColor, value=<rgba 1 1 1 1>}}
  User.HueShift_M3 {type=NiagaraFloat, defaultValue=1, description=""}
  User.MaxGlassColor {type=NiagaraFloat, defaultValue=5, description=""}
  User.MinimumGlassColor {type=NiagaraFloat, defaultValue=0.5, description=""}
  User.PaintedColor {type=LinearColor, description="", defaultValue {struct=LinearColor, value=<rgba 1 1 1 1>}}
  User.PaintedScalar {type=NiagaraFloat, defaultValue=1, description=""}
  User.ParticleColorVar {type=NiagaraFloat, defaultValue=1, description=""}
  User.PortCount {type=NiagaraInt32, defaultValue=1}
  User.Power_AlphaPaintedR4 {type=NiagaraFloat, defaultValue=1, description=""}
  User.Power_EmissivePaintedR4 {type=NiagaraFloat, defaultValue=1, description=""}
  User.Shape01 {type=LinearColor, description="", defaultValue {struct=LinearColor, value=<rgba 0 0.7784 1 1>}}
  User.SingleBoostPos {type=NiagaraPosition, description="", defaultValue {}}
  User.WaveMaxScale {type=NiagaraFloat, defaultValue=1, description=""}
  User.WavePeriod {type=NiagaraFloat, defaultValue=1, description=""}
 ]
}
